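# Physics asset for the Neon Animal backpack cosmetic, giving its hanging ornament collision and limited swing.
# A large capsule covers the pack bone, and two smaller capsules cover the dyn_ornament_01 and dyn_ornament_02 bones.
# Two constraints chain the second ornament to the first and the first to the pack, each allowing moderate swing with a tight twist limit.
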PhysicsAsset /BRCosmetics/Accessories/FORT_Backpacks/Backpack_M_MED_Neon_Animal/Meshes/M_MED_Neon_Animal_Pack_Physics
bodies {
 dyn_ornament_01 [{
   shapeType=Capsule
   length=5.0877
   radius=1.0473
   shapeName=dyn_ornament_01_capsule
   center=<2.8958 0 -0.1244>
   rotation=<rot -89.1015 0.0001 -0.0001>
  }]
 dyn_ornament_02 [{
   shapeType=Capsule
   length=4.7693
   radius=3.0344
   shapeName=dyn_ornament_02_capsule
   center=<3.9509 1.249 2.6835>
   rotation=<rot -53.5618 0 19.7691>
  }]
 pack [{shapeType=Capsule, length=13.6213, radius=16.3787, center=<-3.8383 10.9257 -5.3328>}]
}
constraints [
 {bone1Name=dyn_ornament_02, bone2Name=dyn_ornament_01, swing1LimitDegrees=45, swing2LimitDegrees=45, twistLimitDegrees=5}
 {bone1Name=dyn_ornament_01, bone2Name=pack, swing1LimitDegrees=45, swing2LimitDegrees=45, twistLimitDegrees=5}
]
